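; Boost_Phoenix is a vehicle cosmetic actor for the Phoenix boost effect.
; When the color-ready event fires, it passes the color name, the additional info and its NS_Boost component to the shared cosmetics function library, using "color" as the parameter name.
; BeginPlay, overlap and Tick events are empty, and the construction script only calls the parent.
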(event EventBeginPlay)

(event Collision|EventActorBeginOverlap
  (params (OtherActor :type Actor%20Object%20Reference)))

(event EventTick
  (params (DeltaSeconds :type Float%20%28single-precision%29)))

(event Vehicle|Cosmetics|EventOnColorReady
  (params (Color :type Name :default "None") (AdditionalInfo :type Vehicle%20Cosmetics%20Additional%20Parameter%20Infos%20Structure))
  (Class|BPVehicleCosmeticsFunctionLibrary|CosmeticActorColorReady Color "color" AdditionalInfo (Variables|Default|GetNS_Boost)))

(fn ConstructionScript
  (|CallParentFunction))
