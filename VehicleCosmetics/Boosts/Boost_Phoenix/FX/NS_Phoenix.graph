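# Niagara system for the Phoenix vehicle boost cosmetic: a fiery orange trail with wing meshes.
# A ribbon follows the source with game-mode-scaled spawn rate, while looping bursts form rotating, fading wings.
# Sprite01 uses spawn groups and visibility tags to mix curl-noise meshes and sprites in one emitter.
# User.bIsFrontendPreview retunes rates, offsets and speeds, and enables Sprite02's extra glow and dust; painted-variant overrides are disabled.
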NiagaraSystem /VehicleCosmetics/Boosts/Boost_Phoenix/FX/NS_Phoenix
dependencies {
 usedDataInterfaces [
  &NiagaraDataInterfaceCurve
  &NiagaraDataInterfaceArrayInt32
  &NiagaraDataInterfaceSkeletalMesh
  &NiagaraDataInterfaceArrayFloat3
 ]
 usedDynamicInputs [
  &Multiply_Vector
  &AddVectorToPosition
  /Niagara/DynamicInputs/TypeConversions/MakeCustomVector4FromBool
  &SelectVectorFromArray
 ]
 usedModules [
  &self:SystemSpawnScript.SystemScriptSource.SystemScriptGraph.NiagaraNodeAssignment_0.SetVariables_#1
  &self:SystemSpawnScript.SystemScriptSource.SystemScriptGraph.NiagaraNodeAssignment_1.SetVariables_#2
  &self:SystemSpawnScript.SystemScriptSource.SystemScriptGraph.NiagaraNodeAssignment_3.SetVariables_#3
  &self:SystemSpawnScript.SystemScriptSource.SystemScriptGraph.NiagaraNodeAssignment_2.SetVariables_#4
  &SystemState
 ]
}
emitters {
 NE_VC_SingleBoost_Ribbon {
  emitterUpdateScript [
   &EmitterState()
   &VC_FilteredSocketCount()
   &VC_RateQualityHandler(Default Rate:NiagaraFloat=&Multiply_Float(
     A=50
     B=/VehicleCosmetics/FX/Niagara/Modules/VC_Mask_Floats(Coro Value:NiagaraFloat=1, FrontEnd Value:NiagaraFloat=0.75, FN Value:NiagaraFloat=0.5)
    ))
   &VC_SpawnRibbon_v1(SpawnRate=$Emitter.Spawn Rate)
   /SetVariables(Emitter.Opacity=3.5) {isSetParameters=true}
  ]
  particleSpawnScript [
   &VC_PaintedVariantsOverride() {enabled=false}
   &VC_InitializeRibbon_v1(
    Color=<rgba 1 1 1 5>
    Custom Offset=true
    Lifetime=&MakeCustomFloatFromBool(Bool=$User.bIsFrontendPreview, True Float=2, False Float=0.75)
    Local Vector Offset=/Niagara/DynamicInputs/TypeConversions/MakeCustomPositionFromBool(Bool:NiagaraBool=$User.bIsFrontendPreview, True Position:NiagaraPosition=<-150 0 30>, False Position:NiagaraPosition=<0 0 0>)
    Offset Amount=1
    Ribbon Width:NiagaraFloat=&MakeCustomFloatFromBool(Bool=$User.bIsFrontendPreview, True Float=45, False Float=55)
   )
   &AddVelocity(Velocity=&VC_Mask_Vectors(Coro Value:Vector3f=<-400 0 0>, FrontEnd Value:Vector3f=<-1200 0 0>, FN Value:Vector3f=<-1200 0 0>)) {module=AddVelocity001}
   /VehicleCosmetics/FX/VC_FXParents/Modules/DLR_ProductPreviews(GameModeID:NiagaraInt32=$System.GameModeID, bPreviewVehicleSpeed:NiagaraBool=$System.bPreviewVehicleSpeed)
  ]
  particleUpdateScript [
   &ParticleState()
   &InheritSourceMovement(
    Applied Position Scale=&VectorFromFloat(Value=&FloatFromCurve(FloatCurve=DataInterface() {propertyValues {Curve {
         defaultValue=1
         keys [
          {arriveTangent=-1.7885, interpMode=RCIM_Cubic, leaveTangent=-1.7885, tangentMode=RCTM_User, time=0.05, value=1}
          {interpMode=RCIM_Cubic, tangentMode=RCTM_User, time=1, value=0}
         ]
        }}}))
    ApplyForce=false
    ApplyPosition=true
    ApplyPositionFalloff=false
   )
   &SolveForcesAndVelocity(Clamp Velocity=false, Limit Acceleration=false)
   /SetVariables(Particles.New Color:LinearColor=<rgba 50 20.0574 3.943 1>) {isSetParameters=true}
   &VC_UpdateRibbon_v1()
   &ScaleRibbonWidth(Ribbon Width Scale=&FloatFromCurve(FloatCurve=DataInterface() {propertyValues {Curve {keys [<at 0 1.5 RCIM_Cubic>, {arriveTangent=-5.596, interpMode=RCIM_Cubic, leaveTangent=-5.596, tangentMode=RCTM_User, time=1, value=0}]}}}))
  ]
  rendererClasses [&NiagaraRibbonRendererProperties]
  renderers [
   {rendererClass=&NiagaraRibbonRendererProperties}
   {rendererClass=&NiagaraRibbonRendererProperties, rendererIndex=1}
  ]
 }
 NE_VC_SingleBoost_Sprite01 {
  emitterUpdateScript [
   &EmitterState()
   &SpawnRate(
    SpawnRate=&MakeCustomFloatFromBool(Bool=$User.bIsFrontendPreview, True Float=60, False Float=200)
    Use Spawn Probability=false
   )
   &SpawnRate(
    SpawnGroup=1
    SpawnRate=&MakeCustomFloatFromBool(Bool=$User.bIsFrontendPreview, True Float=0, False Float=100)
    Use Spawn Probability=false
   ) {module=SpawnRate001}
   &SpawnRate(
    SpawnGroup=2
    SpawnRate=&MakeCustomFloatFromBool(Bool=$User.bIsFrontendPreview, True Float=200, False Float=200)
    Use Spawn Probability=false
   ) {module=SpawnRate002}
   /SetVariables(Emitter.Color:LinearColor=&MakeCustomLinearColorFromBool(Bool=$User.bIsFrontendPreview, True Color=<rgba 3 0.65 0.18 1>, False Color=<rgba 30 3.81 0.9 3>)) {isSetParameters=true}
  ]
  particleSpawnScript [
   &VC_PaintedVariantsOverride() {enabled=false}
   &InitializeParticle(
    Color=&MergeHelper_ColorSwitch(
     0=<rgba 0 0 0 1>
     1=<rgba 1 1 1 1>
     2=<rgba 100 1.549 0 1>
     3=<rgba 1 1 1 1>
     Index=&SpawnGroup()
     Num=::NewEnumerator2
    )
    Lifetime=&MergeHelper_FloatSwitch(
     0=&MakeCustomFloatFromBool(
      Bool=$User.bIsFrontendPreview
      True Float=&RandomRangeFloat(Minimum=0.5, Maximum=1.5)
      False Float=&RandomRangeFloat(Minimum=0.2, Maximum=0.8)
     )
     1=&RandomRangeFloat(Minimum=0.3, Maximum=0.6)
     2=&RandomRangeFloat(Minimum=0.9, Maximum=0.25)
     3=&RandomRangeFloat(Minimum=1, Maximum=0.6)
     Index=&SpawnGroup()
     Num=::NewEnumerator2
    )
    Mesh Scale=&MakeCustomPositionFromBool(
     Bool:NiagaraBool=$User.bIsFrontendPreview
     True Position:NiagaraPosition=&RandomRangeVector(Minimum=<6 5 4>, Maximum=<10 5 5>)
     False Position:NiagaraPosition=&RandomRangeVector(Minimum=<8 8 5>, Maximum=<30 10 5>)
    )
    Mesh Scale Mode=::NewEnumerator4
    Position Offset=&MakeCustomPositionFromBool(Bool:NiagaraBool=$User.bIsFrontendPreview, True Position:NiagaraPosition=<-200 0 25>, False Position:NiagaraPosition=<0 0 0>)
    Position Offset Coordinate Space=::Local
    Sprite Rotation Mode=::NewEnumerator2
    Sprite Size Mode=::NewEnumerator3
    Uniform Sprite Size=&MergeHelper_FloatSwitch(
     0=0
     1=&RandomRangeFloat(Minimum=150, Maximum=250)
     2=&MakeCustomFloatFromBool(
      Bool=$User.bIsFrontendPreview
      True Float=&RandomRangeFloat(Minimum=5, Maximum=20)
      False Float=&RandomRangeFloat(Minimum=10, Maximum=60)
     )
     3=&RandomRangeFloat(Minimum=150, Maximum=50)
     Index=&SpawnGroup()
     Num=::NewEnumerator2
    )
    UsePositionOffset=true
    Write Material Random=true
   )
   &AddVelocity(
    Cone Angle=&MergeHelper_FloatSwitch(
     0=&MakeCustomFloatFromBool(Bool=$User.bIsFrontendPreview, True Float=1, False Float=20)
     1=&MakeCustomFloatFromBool(Bool=$User.bIsFrontendPreview, True Float=1, False Float=40)
     2=&MakeCustomFloatFromBool(Bool=$User.bIsFrontendPreview, True Float=1, False Float=80)
     3=1
     Index=&SpawnGroup()
     Num=::NewEnumerator2
    )
    Module.UseConeFalloff=false
    Module.UseVelDistribution=false
    Override Seed=false
    Velocity Mode=::NewEnumerator2
    Velocity Speed=&MergeHelper_FloatSwitch(
     0=&MakeCustomFloatFromBool(
      Bool=$User.bIsFrontendPreview
      True Float=&RandomRangeFloat(Minimum=-2000, Maximum=-1200)
      False Float=&RandomRangeFloat(Minimum=0, Maximum=-20)
     )
     1=&MakeCustomFloatFromBool(
      Bool=$User.bIsFrontendPreview
      True Float=&RandomRangeFloat(Minimum=-2000, Maximum=-3000)
      False Float=&RandomRangeFloat(Minimum=-1200, Maximum=-500)
     )
     2=&MakeCustomFloatFromBool(
      Bool=$User.bIsFrontendPreview
      True Float=&RandomRangeFloat(Minimum=-2000, Maximum=-3000)
      False Float=&RandomRangeFloat(Minimum=-10, Maximum=-500)
     )
     3=?
     Index=&SpawnGroup()
     Num=::NewEnumerator1
    )
   )
   &ShapeLocation(
    MaskSpawnGroup=false
    Override Seed=false
    Sphere Radius=&MergeHelper_FloatSwitch(
     0=&MakeCustomFloatFromBool(Bool=$User.bIsFrontendPreview, True Float=50, False Float=50)
     1=&MakeCustomFloatFromBool(Bool=$User.bIsFrontendPreview, True Float=15, False Float=60)
     2=&MakeCustomFloatFromBool(Bool=$User.bIsFrontendPreview, True Float=70, False Float=100)
     3=15
     Index=&SpawnGroup()
     Num=::NewEnumerator2
    )
    UseOwnerScale=true
   )
   &InitialMeshOrientation(Rotation=&RandomRangeVector(Minimum=<-1 -1 -1>, Maximum=<1 1 1>), Use Rotation Vector=true)
   /SetVariables(Particles.VisibilityTag=&SelectIntFromArray(Direct Array Index=&SpawnGroup(), Int Selection Array=DataInterface() {propertyValues {IntData [1,0,2,3]}})) {isSetParameters=true}
  ]
  particleUpdateScript [
   &ParticleState()
   &ScaleSpriteSize(
    Scale Sprite Size Mode=::NewEnumerator0
    Uniform Scale Factor=&MergeHelper_FloatSwitch(
     0=0
     1=&FloatFromCurve(FloatCurve=DataInterface() {propertyValues {Curve {defaultValue=1, keys [{arriveTangent=-0.5, interpMode=RCIM_Cubic, time=0.25, value=1}, {tangentMode=RCTM_User, time=1, value=0}]}}})
     2=&FloatFromCurve(FloatCurve=DataInterface() {propertyValues {Curve {
         defaultValue=1
         keys [{interpMode=RCIM_Cubic, tangentMode=RCTM_User, time=0, value=0}, <at 0.15 1 RCIM_Cubic>, {tangentMode=RCTM_User, time=1, value=0}]
        }}})
     3=&FloatFromCurve(FloatCurve=DataInterface() {propertyValues {Curve {defaultValue=1, keys [{arriveTangent=-0.5, interpMode=RCIM_Cubic, time=0.25, value=1}, {tangentMode=RCTM_User, time=1, value=0}]}}})
     Index=&SpawnGroup()
     Num=::NewEnumerator2
    )
   )
   &ScaleMeshSize(Scale Factor=&VectorFromFloat(Value=&FloatFromCurve(
      FloatCurve=DataInterface() {propertyValues {Curve {
         defaultValue=1
         keys [
          {arriveTangent=6.2807, interpMode=RCIM_Cubic, leaveTangent=6.2807, tangentMode=RCTM_User, time=0.15, value=0.5}
          {arriveTangent=-0.1576, interpMode=RCIM_Cubic, leaveTangent=-0.1576, tangentMode=RCTM_User, time=0.25, value=1}
          <at 0.5 1 RCIM_Cubic>
         ]
        }}}
      Scale Curve=&MakeCustomFloatFromBool(Bool=$User.bIsFrontendPreview, True Float=1, False Float=&RandomRangeFloat(Minimum=1.5, Maximum=1))
     )))
   &UpdateMeshOrientation(
    Rotation Rate=&RandomRangeFloat(Minimum=0.2, Maximum=1)
    Rotation Vector=&RandomRangeVector(Minimum=<-1 -1 -1>, Maximum=<1 1 1>)
   )
   &CurlNoiseForce(
    Bias Noise=false
    Noise Frequency=600
    Noise Strength=&MergeHelper_FloatSwitch(
     0=&MakeCustomFloatFromBool(Bool=$User.bIsFrontendPreview, True Float=1500, False Float=3000)
     1=0
     2=3000
     3=0
     Index=&SpawnGroup()
     Num=::NewEnumerator2
    )
    Pan Noise=false
   )
   &ScaleColor(
    Scale Alpha=&MergeHelper_FloatSwitch(
     0=1
     1=&FloatFromCurve(
      FloatCurve=DataInterface() {propertyValues {Curve {
         defaultValue=1
         keys [{interpMode=RCIM_Cubic, tangentMode=RCTM_User, time=0, value=0}, <at 0.18 1 RCIM_Cubic>, {tangentMode=RCTM_User, time=1, value=0}]
        }}}
      Scale Curve=0.5
     )
     2=&FloatFromCurve(FloatCurve=DataInterface() {propertyValues {Curve {
         defaultValue=1
         keys [
          {interpMode=RCIM_Cubic, tangentMode=RCTM_User, time=0, value=0}
          {interpMode=RCIM_Cubic, tangentMode=RCTM_User, time=0.5, value=1}
          {tangentMode=RCTM_User, time=1, value=0}
         ]
        }}})
     3=&FloatFromCurve(
      FloatCurve=DataInterface() {propertyValues {Curve {
         defaultValue=1
         keys [{interpMode=RCIM_Cubic, tangentMode=RCTM_User, time=0, value=0}, <at 0.18 1 RCIM_Cubic>, {tangentMode=RCTM_User, time=1, value=0}]
        }}}
      Scale Curve=0.5
     )
     Index=&SpawnGroup()
     Num=::NewEnumerator2
    )
    ScaleA=true
    ScaleRGB=true
   )
   &DynamicMaterialParameters(
    Index 0 Param 1=&MergeHelper_FloatSwitch(0=0, 1=0, 2=0.5, 3=0, Index=&SpawnGroup(), Num=::NewEnumerator2)
    Index 0 Param 2=&MergeHelper_FloatSwitch(0=1, 1=1, 2=1, 3=1, Index=&SpawnGroup(), Num=::NewEnumerator2)
    Index 0 Param 3=&MergeHelper_FloatSwitch(0=0, 1=0.1, 2=3, 3=0.1, Index=&SpawnGroup(), Num=::NewEnumerator2)
    Index 0 Param 4=&MergeHelper_FloatSwitch(
     0=&FloatFromCurve(
      FloatCurve=DataInterface() {propertyValues {Curve {defaultValue=1, keys [{tangentMode=RCTM_User, time=0.45, value=0}, {tangentMode=RCTM_User, time=1, value=1}]}}}
      Scale Curve=&RandomRangeFloat(Minimum=0.5, Maximum=1)
     )
     1=1
     2=0.5
     3=1
     Index=&SpawnGroup()
     Num=::NewEnumerator2
    )
    Index 1 Param 1=&MergeHelper_FloatSwitch(0=0, 1=1, 2=0.25, 3=1, Index=&SpawnGroup(), Num=::NewEnumerator2)
    Index 1 Param 2=&MergeHelper_FloatSwitch(0=1, 1=0, 2=50, 3=0, Index=&SpawnGroup(), Num=::NewEnumerator2)
    Index 1 Param 3=&MergeHelper_FloatSwitch(0=0, 1=1, 2=0, 3=1, Index=&SpawnGroup(), Num=::NewEnumerator2)
    Index 2 Param 1=-3.7644
    Index 2 Param 2=1
    Module.Index0Float0Write=true
    Module.Index0Float1Write=true
    Module.Index0Float2Write=true
    Module.Index0Float3Write=true
    Module.Index1Float0Write=true
    Module.Index1Float1Write=true
    Module.Index1Float2Write=true
    Module.Index1Float3Write=true
    Module.Index2Float0Write=true
    Module.Index2Float1Write=true
    Module.Index2Float2Write=true
    Module.Index2Float3Write=true
    Write Parameter Index 1=::NewEnumerator1
    Write Parameter Index 2=::NewEnumerator1
   )
   /SetVariables(Particles.SubImageIndex=&MergeHelper_FloatSwitch(
     0=0
     1=&RandomRangeFloat(Minimum=0, Maximum=3)
     2=0
     3=&RandomRangeFloat(Minimum=0, Maximum=3)
     Index=&SpawnGroup()
     Num=::NewEnumerator2
    )) {isSetParameters=true}
   &SolveForcesAndVelocity(Clamp Velocity=false, Limit Acceleration=false)
  ]
  rendererClasses [&NiagaraSpriteRendererProperties, &NiagaraMeshRendererProperties]
  renderers [
   {rendererClass=&NiagaraSpriteRendererProperties}
   {rendererClass=&NiagaraMeshRendererProperties, rendererIndex=1}
   {rendererClass=&NiagaraSpriteRendererProperties, rendererIndex=2}
  ]
 }
 NE_VC_SingleBoost_Sprite02 {
  emitterUpdateScript [
   &EmitterState()
   &SpawnRate(
    SpawnRate=&MakeCustomFloatFromBool(Bool=$User.bIsFrontendPreview, True Float=30, False Float=0)
    Use Spawn Probability=false
   )
  ]
  particleSpawnScript [
   &VC_PaintedVariantsOverride() {enabled=false}
   &InitializeParticle(
    Color=<rgba 2 0.4333 0 1.1>
    Lifetime=&RandomRangeFloat(Minimum=1, Maximum=1)
    Position Offset=&MakeCustomPositionFromBool(Bool:NiagaraBool=$User.bIsFrontendPreview, True Position:NiagaraPosition=<-200 0 25>, False Position:NiagaraPosition=<0 0 0>)
    Position Offset Coordinate Space=::Local
    Sprite Rotation Mode=::NewEnumerator2
    Sprite Size Mode=::NewEnumerator3
    Uniform Sprite Size=&RandomRangeFloat(Minimum=170, Maximum=100)
    UsePositionOffset=true
    Write Material Random=true
   )
   &AddVelocity(
    Cone Angle=0
    Module.UseConeFalloff=false
    Module.UseVelDistribution=false
    Override Seed=false
    Velocity Mode=::NewEnumerator2
    Velocity Speed=&MakeCustomFloatFromBool(
     Bool=$User.bIsFrontendPreview
     True Float=&RandomRangeFloat(Minimum=-2000, Maximum=-3000)
     False Float=&RandomRangeFloat(Minimum=-10, Maximum=-500)
    )
   )
   &ShapeLocation(MaskSpawnGroup=false, Override Seed=false, Sphere Radius=15, UseOwnerScale=true)
  ]
  particleUpdateScript [
   &ParticleState()
   &SpriteRotationRate(Rotation Rate=&RandomRangeFloat(Minimum=-1, Maximum=1))
   &ScaleSpriteSize(
    Scale Sprite Size Mode=::NewEnumerator0
    Uniform Scale Factor=&FloatFromCurve(FloatCurve=DataInterface() {propertyValues {Curve {
        defaultValue=1
        keys [
         {arriveTangent=-1.1111, interpMode=RCIM_Cubic, time=0, value=0}
         {arriveTangent=-1.4308, interpMode=RCIM_Cubic, leaveTangent=-1.4308, tangentMode=RCTM_User, time=0.05, value=1.05}
         {tangentMode=RCTM_User, time=0.9, value=0}
        ]
       }}})
   )
   &ScaleColor(
    Scale Alpha=&FloatFromCurve(
     FloatCurve=DataInterface() {propertyValues {Curve {
        defaultValue=1
        keys [{interpMode=RCIM_Cubic, tangentMode=RCTM_User, time=0, value=0}, <at 0.05 1 RCIM_Cubic>, {tangentMode=RCTM_User, time=1, value=0}]
       }}}
     Scale Curve=0.5
    )
    ScaleA=true
    ScaleRGB=true
   )
   &SolveForcesAndVelocity(Clamp Velocity=false, Limit Acceleration=false)
   /SetVariables(
    Particles.Dust Color:LinearColor=&MakeLinearColorFromVectorAndFloat(
     "Vector (RGB)"=&MakeVectorFromLinearColorRGB(Color=&RandomRangeLinearColor(Minimum=<rgba 50 6.6667 0 1>, Maximum=<rgba 30 9.75 3 1>))
     "Float (Alpha)"=&FloatFromCurve(
      FloatCurve=DataInterface() {propertyValues {Curve {
         defaultValue=1
         keys [
          {interpMode=RCIM_Cubic, tangentMode=RCTM_User, time=0.05, value=0}
          {interpMode=RCIM_Cubic, tangentMode=RCTM_User, time=0.15, value=1}
          {tangentMode=RCTM_User, time=0.5, value=0}
         ]
        }}}
      Scale Curve=&Lerp_Float(A=0, B=1, Alpha=0.5)
     )
    )
    Particles.Dust Size:Vector2f=&Vector2DFromFloat(Value=&FloatFromCurve(
      FloatCurve=DataInterface() {propertyValues {Curve {
         defaultValue=1
         keys [
          {interpMode=RCIM_Cubic, tangentMode=RCTM_User, time=0, value=0}
          {interpMode=RCIM_Cubic, tangentMode=RCTM_User, time=0.1, value=1}
          {tangentMode=RCTM_User, time=0.5, value=1}
         ]
        }}}
      Scale Curve=&RandomRangeFloat(Minimum=130, Maximum=60)
     ))
   ) {isSetParameters=true}
  ]
  rendererClasses [&NiagaraSpriteRendererProperties]
  renderers [
   {rendererClass=&NiagaraSpriteRendererProperties}
   {rendererClass=&NiagaraSpriteRendererProperties, rendererIndex=1}
   {rendererClass=&NiagaraSpriteRendererProperties, rendererIndex=2}
  ]
 }
 NE_VC_SingleBoost_Wings {
  emitterUpdateScript [
   &EmitterState(Life Cycle Mode=::NewEnumerator1, Loop Behavior=::NewEnumerator1, UseLoopDelay=false)
   &SpawnBurst_Instantaneous(Spawn Count=6, Use Spawn Probability=false, UseLoopCountLimit=false)
   &SpawnBurst_Instantaneous(Spawn Count=6, Spawn Group=1, Use Spawn Probability=false, UseLoopCountLimit=false) {module=SpawnBurst_Instantaneous001}
   &SpawnBurst_Instantaneous(Spawn Count=6, Spawn Group=2, Use Spawn Probability=false, UseLoopCountLimit=false) {module=SpawnBurst_Instantaneous002}
   &SpawnRate(SpawnRate=50, Use Spawn Probability=false) {enabled=false}
  ]
  particleSpawnScript [
   &VC_PaintedVariantsOverride() {enabled=false}
   &InitializeParticle(
    Color=<rgba 1000 16.6668 0.0001 1>
    Lifetime Max=0.45
    Lifetime Min=0.3
    Lifetime Mode=::NewEnumerator1
    Mesh Scale Mode=::NewEnumerator1
    Mesh Uniform Scale Max=7
    Mesh Uniform Scale Min=3
    Position Offset=&RandomRangeVector(Minimum=<-210 0 25>, Maximum=<-210 0 20>)
    Position Offset Coordinate Space=::Local
    Sprite Rotation Mode=::NewEnumerator2
    UsePositionOffset=true
    Write Material Random=true
   )
   &AddVelocity()
   &InitialMeshOrientation(Mesh Axis To Orient=<0 0 1>, Mesh Orientation Mode=::NewEnumerator2, Rotation=<0.25 0.5 0>, Use Rotation Vector=true)
   &InitialMeshOrientation(
    Mesh Axis To Orient=&MergeHelper_VectorSwitch(0=<1 0 0.35>, 1=<1 1 0>, 2=?, 3=?, Index=&SpawnGroup(), Num=::NewEnumerator0)
    Mesh Orientation Mode=::NewEnumerator2
    Rotation=&MergeHelper_VectorSwitch(
     0=&RandomRangeVector(Minimum=<0 0 0>, Maximum=<0.02 0 0>)
     1=&RandomRangeVector(Minimum=<0 0 -0.25>, Maximum=<0 0.02 -0.25>)
     2=?
     3=?
     Index=&SpawnGroup()
     Num=::NewEnumerator0
    )
    Use Rotation Vector=true
   ) {module=InitialMeshOrientation001}
   /SetVariables(Particles.VisibilityTag=$Emitter.SpawnGroup) {isSetParameters=true}
   &CameraOffset(Camera Offset Amount=50)
  ]
  particleUpdateScript [
   &ParticleState()
   &UpdateMeshOrientation(Rotation Rate=&MergeHelper_FloatSwitch(0=0.1, 1=-0.1, 2=?, 3=?, Index=&SpawnGroup(), Num=::NewEnumerator0))
   &ScaleMeshSize(Scale Factor=&VectorFromFloat(Value=&FloatFromCurve(FloatCurve=DataInterface() {propertyValues {Curve {defaultValue=1, keys [{tangentMode=RCTM_User, time=0, value=1}, <at 0.2 1.2>, {tangentMode=RCTM_User, time=1, value=1.35}]}}})))
   &SolveForcesAndVelocity(Clamp Velocity=false, Limit Acceleration=false)
   &DynamicMaterialParameters(
    Index 0 Param 1=1
    Index 0 Param 2=1
    Index 0 Param 4=&FloatFromCurve(FloatCurve=DataInterface() {propertyValues {Curve {
        defaultValue=-0.5
        keys [
         {arriveTangent=1.7777, interpMode=RCIM_Cubic, leaveTangent=1.7777, tangentMode=RCTM_User, time=0, value=-1}
         {arriveTangent=0.4306, interpMode=RCIM_Cubic, leaveTangent=0.4306, tangentMode=RCTM_User, time=0.3522, value=-0.1647}
         {arriveTangent=0.0777, interpMode=RCIM_Cubic, leaveTangent=2, tangentMode=RCTM_Break, time=0.8, value=0}
         {arriveTangent=0.061, interpMode=RCIM_Cubic, leaveTangent=0.061, tangentMode=RCTM_User, time=1, value=1}
        ]
       }}})
    Index 3 Param 2=1
    Module.Index0Float0Write=true
    Module.Index0Float1Write=true
    Module.Index0Float2Write=true
    Module.Index0Float3Write=true
    Module.Index3Float0Write=true
    Module.Index3Float1Write=true
    Module.Index3Float2Write=true
    Module.Index3Float3Write=true
    Write Parameter Index 3=::NewEnumerator1
   )
   &ScaleColor(
    Scale Alpha=&FloatFromCurve(FloatCurve=DataInterface() {propertyValues {Curve {keys [{arriveTangent=-2.1694, interpMode=RCIM_Cubic, leaveTangent=-2.1694, tangentMode=RCTM_User, time=0, value=1}, <at 1 0 RCIM_Cubic>]}}})
    Scale RGB=<10 10 10>
    ScaleA=true
    ScaleRGB=true
   )
  ]
  rendererClasses [&NiagaraSpriteRendererProperties, &NiagaraMeshRendererProperties]
  renderers [
   {rendererClass=&NiagaraSpriteRendererProperties}
   {rendererClass=&NiagaraMeshRendererProperties, rendererIndex=1}
   {rendererClass=&NiagaraMeshRendererProperties, rendererIndex=2}
  ]
 }
}
summary {
 systemName=NS_Phoenix
 emitters [
  {emitterName=NE_VC_SingleBoost_Ribbon, rendererClasses [&NiagaraRibbonRendererProperties]}
  {
   emitterName=NE_VC_SingleBoost_Wings
   rendererClasses [&NiagaraSpriteRendererProperties, &NiagaraMeshRendererProperties]
  }
  {
   emitterName=NE_VC_SingleBoost_Sprite01
   rendererClasses [&NiagaraSpriteRendererProperties, &NiagaraMeshRendererProperties]
  }
  {emitterName=NE_VC_SingleBoost_Sprite02, rendererClasses [&NiagaraSpriteRendererProperties]}
 ]
 userVariables [
  User.bIsFrontendPreview {type=NiagaraBool}
  User.bIsPainted {type=NiagaraBool}
  User.bPreviewVehicleSpeed {type=NiagaraBool}
  User.ColorVariantID {type=NiagaraInt32}
  User.DTCarOffset {type=Vector3f, description="", defaultValue {struct=Vector3f, value=<0 0 0>}}
  User.GameModeID {type=NiagaraInt32}
  User.PaintedColor {type=LinearColor, description="", defaultValue {struct=LinearColor, value=<rgba 1 1 1 1>}}
  User.PaintedScalar {type=NiagaraFloat, defaultValue=1, description=""}
  User.PortCount {type=NiagaraInt32, defaultValue=1}
  User.SingleBoostPos {type=NiagaraPosition, description="", defaultValue {}}
 ]
}
